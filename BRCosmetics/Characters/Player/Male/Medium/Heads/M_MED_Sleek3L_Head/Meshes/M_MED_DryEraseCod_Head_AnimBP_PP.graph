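# Post-process AnimBP for the DryEraseCod head on the Fortnite 3L head skeleton, applying an additive bone-space rotation to the head bone.
# For tagged characters on a pawn outside the front end, update stores scaled, negated head-versus-neck_02 ref-pose delta in head rot, yaw-only during emotes.
# The Modify Bone node shows a fixed rotation, with no visible link to head rot.
# Init and customization load set the tag gate and NoLimitHead, which is never read here.
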AnimBlueprint /BRCosmetics/Characters/Player/Male/Medium/Heads/M_MED_Sleek3L_Head/Meshes/M_MED_DryEraseCod_Head_AnimBP_PP
target_skeleton=/Game/Characters/Player/Common/Fortnite_Base_Head/Fortnite_Base_3L_Head_Skeleton
graphs [
 AnimGraph {nodes [
   n1 = "Transform (Modify) Bone - Bone: head"!(
    ComponentPose=Local To Component(LocalPose=Input Pose())
    bAlphaBoolEnabled:bool=True
    Translation="0,0,0"
    Rotation="50,50,50"
    Scale="1,1,1"
   ) -> (Pose) {
    alphaInputType=Bool
    rotationMode=BMM_Additive
    rotationSpace=BCS_BoneSpace
    alphaBoolBlend {blendInTime=0.1, blendOption=Linear, blendOutTime=0.1}
    boneToModify {boneName=head}
    rotation=<rot 50 50 50>
   }
   Output Pose(Result=Component To Local(ComponentPose=@n1))
  ]}
 EventGraph {nodes [
   Event BlueprintUpdateAnimation()
   > Branch(Condition=Get TagInit() -> (TagInit:bool))
   > Branch(Condition=Get bIsOnPawn() -> (bIsOnPawn:bool))
   > Branch(Condition=NOT Boolean(A=Get bIsInFrontEnd()))
   > n1 = Branch(Condition=NOT Boolean(A=Get bIsPlayingEmote()))
   n2 = BreakTransform(InTransform=GetDeltaTransformFromRefPose(self=GetOwnerSkeletalMeshComponent(), BoneName=head, BaseName=neck_02))
   n3 = BreakRotator(InRot=@n2.Rotation)
   n4 = Get Y percentage() -> (Y percentage:float)
   n5 = Get Z percentage() -> (Z percentage:float)
   n6 = Get X percentage() -> (X percentage:float)
   > Set head rot(head rot:Rotator=MakeRotator(
     Roll=float * float!(A=@n3, B=-1, C:double=@n6) -> (ReturnValue)
     Pitch=float * float(A=@n3.Pitch, B=-1, C=@n4)
     Yaw=float * float(A=@n3.Yaw, B=-1, C=@n5)
    )) -> (Output_Get:Rotator)
   n7 = BreakRotator(InRot=@n2.Rotation)
   Set head rot(execute=@n1.else, head rot=MakeRotator(Yaw=float * float(A=@n7.Yaw, B=-1, C=0.3)))
   n8 = Event OnCharacterCustomizationLoaded()
   n9 = Sequence!(execute=[@n8.then, @n12.then]) -> (then_0, then_1, then_2:Exec)
   > Set TagInit(TagInit:bool=CharacterItemDefinitionHasAnyMetaTag(Tags=Make Array("[0]"=(TagName="Cosmetics.Metatags.DryEraseCod")))) -> (Output_Get:bool)
   > Set BodyMesh(BodyMesh:Pawn Object Reference=TryGetPawnOwner()) -> (Output_Get:Pawn Object Reference)
   n10 = Get Array Item with EFort Custom Part Type(Array:Array of Custom Character Part Object References=@n8.CustomParts, PartType=Face) -> (Output:Custom Character Part Object Reference)
   n11 = Branch(execute=@n9.then_1, Condition=MatchesWildcard(SourceString=GetDisplayName(Object=@n10), Wildcard=*dameregent*))
   > Set NoLimitHead(NoLimitHead:bool=false) -> (Output_Get:bool)
   Set NoLimitHead(execute=@n11.else, NoLimitHead=true)
   n12 = Event BlueprintInitializeAnimation()
  ]}
]
variables [
 BodyMesh {type=Pawn, default=null}
 head rot {type=Rotator}
 NoLimitHead {type=boolean}
 TagInit {type=boolean}
 X percentage {type=number}
 Y percentage {type=number, default=0.8}
 Z percentage {type=number, default=0.95}
]
